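# Physics asset for the medium female player's Stands PER face-accessory mesh, giving its hair bones capsule bodies and joints around a head capsule.
# One front chain, three side chains per side and three back chains each constrain link to link up to the head.
# Front and back roots swing and twist freely, while side roots and all other links have tight angular limit values.
# Left b and c chains use slimmer, mostly offset capsules, unlike the right.
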PhysicsAsset /Game/Characters/Player/Female/Medium/Bodies/F_MED_Stands/Meshes/Parts/PER/F_MED_Stands_PER_FaceAcc_Physics
bodies {
 hair_a_01_l [{shapeType=Capsule, length=2, radius=2, shapeName=hair_a_01_l_capsule, rotation=<rot 90 0 0>}]
 hair_a_01_r [{shapeType=Capsule, length=2, radius=2, shapeName=hair_a_01_r_capsule, rotation=<rot 90 0 0>}]
 hair_a_02_l [{shapeType=Capsule, length=2, radius=2, shapeName=hair_a_02_l_capsule, rotation=<rot 90 0 0>}]
 hair_a_02_r [{shapeType=Capsule, length=2, radius=2, shapeName=hair_a_02_r_capsule, rotation=<rot 90 0 0>}]
 hair_a_03_l [{shapeType=Capsule, length=2, radius=2, shapeName=hair_a_03_l_capsule, rotation=<rot 90 0 0>}]
 hair_a_03_r [{shapeType=Capsule, length=2, radius=2, shapeName=hair_a_03_r_capsule, rotation=<rot 90 0 0>}]
 hair_a_base_l [{shapeType=Capsule, length=2, radius=2, shapeName=hair_a_base_l_capsule, rotation=<rot 90 0 0>}]
 hair_a_base_r [{shapeType=Capsule, length=2, radius=2, shapeName=hair_a_base_r_capsule, rotation=<rot 90 0 0>}]
 hair_b_01_l [{shapeType=Capsule, length=2.5, radius=1.5, shapeName=hair_b_01_l_capsule, center=<1.1008 -0.0001 0>, rotation=<rot 90 0 0>}]
 hair_b_01_r [{shapeType=Capsule, length=2, radius=2, shapeName=hair_b_01_r_capsule, rotation=<rot 90 0 0>}]
 hair_b_02_l [{shapeType=Capsule, length=2.5, radius=1.5, shapeName=hair_b_02_l_capsule, center=<1.1008 -0.0001 0>, rotation=<rot 90 0 0>}]
 hair_b_02_r [{shapeType=Capsule, length=2, radius=2, shapeName=hair_b_02_r_capsule, rotation=<rot 90 0 0>}]
 hair_b_03_l [{shapeType=Capsule, length=2.5, radius=1.5, shapeName=hair_b_03_l_capsule, center=<1.1008 -0.0001 0>, rotation=<rot 90 0 0>}]
 hair_b_03_r [{shapeType=Capsule, length=2, radius=2, shapeName=hair_b_03_r_capsule, rotation=<rot 90 0 0>}]
 hair_b_base_l [{shapeType=Capsule, length=2.5, radius=1.5, shapeName=hair_b_base_l_capsule, center=<1.1008 -0.0001 0>, rotation=<rot 90 0 0>}]
 hair_b_base_r [{shapeType=Capsule, length=2, radius=2, shapeName=hair_b_base_r_capsule, rotation=<rot 90 0 0>}]
 hair_c_01_l [{shapeType=Capsule, length=2.5, radius=1.5, shapeName=hair_c_01_l_capsule, center=<1.1008 -0.0001 0>, rotation=<rot 90 0 0>}]
 hair_c_01_r [{shapeType=Capsule, length=2, radius=2, shapeName=hair_c_01_r_capsule, rotation=<rot 90 0 0>}]
 hair_c_02_l [{shapeType=Capsule, length=2.5, radius=1.5, shapeName=hair_c_02_l_capsule, center=<1.1008 -0.0001 0>, rotation=<rot 90 0 0>}]
 hair_c_02_r [{shapeType=Capsule, length=2, radius=2, shapeName=hair_c_02_r_capsule, rotation=<rot 90 0 0>}]
 hair_c_03_l [{shapeType=Capsule, length=2.5, radius=1.5, shapeName=hair_c_03_l_capsule, center=<1.1008 -0.0001 0>, rotation=<rot 90 0 0>}]
 hair_c_03_r [{shapeType=Capsule, length=2, radius=2, shapeName=hair_c_03_r_capsule, rotation=<rot 90 0 0>}]
 hair_c_base_l [{shapeType=Capsule, length=2.5, radius=1.5, shapeName=hair_c_base_l_capsule, center=<0.7684 0 0>, rotation=<rot 90 0 0>}]
 hair_c_base_r [{shapeType=Capsule, length=2, radius=2, shapeName=hair_c_base_r_capsule, rotation=<rot 90 0 0>}]
 hairback_a_01 [{shapeType=Capsule, length=1.5, radius=1.5, shapeName=hairback_a_01_capsule, rotation=<rot 90 0 0>}]
 hairback_a_02 [{shapeType=Capsule, length=1.5, radius=1.5, shapeName=hairback_a_02_capsule, rotation=<rot 90 0 0>}]
 hairback_a_base [{shapeType=Capsule, length=0.5, radius=0.5, shapeName=hairback_a_base_capsule, rotation=<rot 90 0 0>}]
 hairback_b_01 [{shapeType=Capsule, length=2, radius=1, shapeName=hairback_b_01_capsule, center=<0.1684 -0.3516 -0.2294>}]
 hairback_b_02 [{shapeType=Capsule, length=2, radius=1, shapeName=hairback_b_02_capsule, center=<-0.1222 0.876 -1.2492>}]
 hairback_b_base [{shapeType=Capsule, length=0.5, radius=0.5, shapeName=hairback_b_base_capsule, rotation=<rot 90 0 0>}]
 hairback_c_01 [{shapeType=Capsule, length=1.5, radius=1.5, shapeName=hairback_c_01_capsule, rotation=<rot 90 0 0>}]
 hairback_c_02 [{shapeType=Capsule, length=1.5, radius=1.5, shapeName=hairback_c_02_capsule, rotation=<rot 90 0 0>}]
 hairback_c_base [{shapeType=Capsule, length=0.5, radius=0.5, shapeName=hairback_c_base_capsule, rotation=<rot 90 0 0>}]
 hairfront_01 [{shapeType=Capsule, length=2, radius=2, shapeName=hairfront_01_capsule, rotation=<rot 90 0 0>}]
 hairfront_02 [{shapeType=Capsule, length=2, radius=2, shapeName=hairfront_02_capsule, rotation=<rot 90 0 0>}]
 hairfront_03 [{shapeType=Capsule, length=2, radius=2, shapeName=hairfront_03_capsule, rotation=<rot 90 0 0>}]
 hairfront_04 [{shapeType=Capsule, length=2, radius=2, shapeName=hairfront_04_capsule, rotation=<rot 90 0 0>}]
 hairfront_base [{shapeType=Capsule, length=0.5, radius=0.5, shapeName=hairfront_base_capsule, rotation=<rot 90 0 0>}]
 head [{shapeType=Capsule, length=9.1572, radius=6.2101, rotation=<rot 90 0 -25.0006>}]
}
constraints [
 {
  bone1Name=hairfront_base
  bone2Name=head
  swing1LimitDegrees=45
  swing1Motion=Free
  swing2LimitDegrees=45
  swing2Motion=Free
  twistLimitDegrees=45
  twistMotion=Free
 }
 {bone1Name=hairfront_01, bone2Name=hairfront_base, swing1LimitDegrees=5, swing2LimitDegrees=5, twistLimitDegrees=5}
 {bone1Name=hairfront_02, bone2Name=hairfront_01, swing1LimitDegrees=5, swing2LimitDegrees=5, twistLimitDegrees=5}
 {bone1Name=hairfront_03, bone2Name=hairfront_02, swing1LimitDegrees=5, swing2LimitDegrees=5, twistLimitDegrees=5}
 {bone1Name=hairfront_04, bone2Name=hairfront_03, swing1LimitDegrees=5, swing2LimitDegrees=5, twistLimitDegrees=5}
 {bone1Name=hair_a_base_l, bone2Name=head, swing1LimitDegrees=5, swing2LimitDegrees=5, twistLimitDegrees=5}
 {bone1Name=hair_a_01_l, bone2Name=hair_a_base_l, swing1LimitDegrees=5, swing2LimitDegrees=5, twistLimitDegrees=5}
 {bone1Name=hair_a_02_l, bone2Name=hair_a_01_l, swing1LimitDegrees=5, swing2LimitDegrees=5, twistLimitDegrees=5}
 {bone1Name=hair_a_03_l, bone2Name=hair_a_02_l, swing1LimitDegrees=5, swing2LimitDegrees=5, twistLimitDegrees=5}
 {bone1Name=hair_b_base_l, bone2Name=head, swing1LimitDegrees=5, swing2LimitDegrees=5, twistLimitDegrees=5}
 {bone1Name=hair_b_01_l, bone2Name=hair_b_base_l, swing1LimitDegrees=5, swing2LimitDegrees=5, twistLimitDegrees=5}
 {bone1Name=hair_b_02_l, bone2Name=hair_b_01_l, swing1LimitDegrees=5, swing2LimitDegrees=5, twistLimitDegrees=5}
 {bone1Name=hair_b_03_l, bone2Name=hair_b_02_l, swing1LimitDegrees=5, swing2LimitDegrees=5, twistLimitDegrees=5}
 {bone1Name=hair_c_base_l, bone2Name=head, swing1LimitDegrees=5, swing2LimitDegrees=5, twistLimitDegrees=5}
 {bone1Name=hair_c_01_l, bone2Name=hair_c_base_l, swing1LimitDegrees=5, swing2LimitDegrees=5, twistLimitDegrees=5}
 {bone1Name=hair_c_02_l, bone2Name=hair_c_01_l, swing1LimitDegrees=5, swing2LimitDegrees=5, twistLimitDegrees=5}
 {bone1Name=hair_c_03_l, bone2Name=hair_c_02_l, swing1LimitDegrees=5, swing2LimitDegrees=5, twistLimitDegrees=5}
 {bone1Name=hair_a_base_r, bone2Name=head, swing1LimitDegrees=5, swing2LimitDegrees=5, twistLimitDegrees=5}
 {bone1Name=hair_a_01_r, bone2Name=hair_a_base_r, swing1LimitDegrees=5, swing2LimitDegrees=5, twistLimitDegrees=5}
 {bone1Name=hair_a_02_r, bone2Name=hair_a_01_r, swing1LimitDegrees=5, swing2LimitDegrees=5, twistLimitDegrees=5}
 {bone1Name=hair_a_03_r, bone2Name=hair_a_02_r, swing1LimitDegrees=5, swing2LimitDegrees=5, twistLimitDegrees=5}
 {bone1Name=hair_b_base_r, bone2Name=head, swing1LimitDegrees=5, swing2LimitDegrees=5, twistLimitDegrees=5}
 {bone1Name=hair_b_01_r, bone2Name=hair_b_base_r, swing1LimitDegrees=5, swing2LimitDegrees=5, twistLimitDegrees=5}
 {bone1Name=hair_b_02_r, bone2Name=hair_b_01_r, swing1LimitDegrees=5, swing2LimitDegrees=5, twistLimitDegrees=5}
 {bone1Name=hair_b_03_r, bone2Name=hair_b_02_r, swing1LimitDegrees=5, swing2LimitDegrees=5, twistLimitDegrees=5}
 {bone1Name=hair_c_base_r, bone2Name=head, swing1LimitDegrees=5, swing2LimitDegrees=5, twistLimitDegrees=5}
 {bone1Name=hair_c_01_r, bone2Name=hair_c_base_r, swing1LimitDegrees=5, swing2LimitDegrees=5, twistLimitDegrees=5}
 {bone1Name=hair_c_02_r, bone2Name=hair_c_01_r, swing1LimitDegrees=5, swing2LimitDegrees=5, twistLimitDegrees=5}
 {bone1Name=hair_c_03_r, bone2Name=hair_c_02_r, swing1LimitDegrees=5, swing2LimitDegrees=5, twistLimitDegrees=5}
 {
  bone1Name=hairback_a_base
  bone2Name=head
  swing1LimitDegrees=45
  swing1Motion=Free
  swing2LimitDegrees=45
  swing2Motion=Free
  twistLimitDegrees=45
  twistMotion=Free
 }
 {bone1Name=hairback_a_01, bone2Name=hairback_a_base, swing1LimitDegrees=5, swing2LimitDegrees=5, twistLimitDegrees=5}
 {bone1Name=hairback_a_02, bone2Name=hairback_a_01, swing1LimitDegrees=5, swing2LimitDegrees=5, twistLimitDegrees=5}
 {
  bone1Name=hairback_b_base
  bone2Name=head
  swing1LimitDegrees=45
  swing1Motion=Free
  swing2LimitDegrees=45
  swing2Motion=Free
  twistLimitDegrees=45
  twistMotion=Free
 }
 {bone1Name=hairback_b_01, bone2Name=hairback_b_base, swing1LimitDegrees=5, swing2LimitDegrees=5, twistLimitDegrees=5}
 {bone1Name=hairback_b_02, bone2Name=hairback_b_01, swing1LimitDegrees=5, swing2LimitDegrees=5, twistLimitDegrees=5}
 {
  bone1Name=hairback_c_base
  bone2Name=head
  swing1LimitDegrees=45
  swing1Motion=Free
  swing2LimitDegrees=45
  swing2Motion=Free
  twistLimitDegrees=45
  twistMotion=Free
 }
 {bone1Name=hairback_c_01, bone2Name=hairback_c_base, swing1LimitDegrees=5, swing2LimitDegrees=5, twistLimitDegrees=5}
 {bone1Name=hairback_c_02, bone2Name=hairback_c_01, swing1LimitDegrees=5, swing2LimitDegrees=5, twistLimitDegrees=5}
]
